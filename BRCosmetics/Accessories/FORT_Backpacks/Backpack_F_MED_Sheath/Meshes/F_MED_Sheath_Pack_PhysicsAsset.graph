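# Physics asset for the F_MED_Sheath backpack cosmetic, providing bodies and joints for its two daggers, zipper and two straps around a box-shaped pack body.
# Daggers and zipper each hang from a small master body whose joint to the pack has free swing and twist.
# The daggers get tight angular limits against their masters, while the zipper locks one swing axis.
# Each strap is a three-link chain of zero-length capsules with limited swing.
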PhysicsAsset /BRCosmetics/Accessories/FORT_Backpacks/Backpack_F_MED_Sheath/Meshes/F_MED_Sheath_Pack_PhysicsAsset
bodies {
 dyn_dagger_l [{shapeType=Box, extentX=6, extentY=3.3088, extentZ=16.3925, center=<-7.7771 0.0001 0.0001>, rotation=<rot 90 0 0>}]
 dyn_dagger_mstr_l [{shapeType=Box, extentX=1.01, extentY=1.01, extentZ=1.01}]
 dyn_dagger_mstr_r [{shapeType=Box, extentX=1.01, extentY=1.01, extentZ=1.01}]
 dyn_dagger_r [{shapeType=Box, extentX=5.8415, extentY=3.3088, extentZ=16.3925, center=<8.1277 0.0001 0>, rotation=<rot 90 0 0>}]
 dyn_strap_01_l [{shapeType=Capsule, length=0, radius=1, center=<-1.1 0 0>}]
 dyn_strap_01_r [{shapeType=Capsule, length=0, radius=1, center=<0.85 0 0>}]
 dyn_strap_02_l [{shapeType=Capsule, length=0, radius=1, center=<-1.1 0 0>}]
 dyn_strap_02_r [{shapeType=Capsule, length=0, radius=1, center=<0.85 0 0>}]
 dyn_strap_03_l [{shapeType=Capsule, length=0, radius=1, center=<-1.1 0 0>}]
 dyn_strap_03_r [{shapeType=Capsule, length=0, radius=1, center=<0.85 0 0>}]
 dyn_zipper [{shapeType=Box, extentX=1.2145, extentY=0.5, extentZ=1.4119, center=<-1 0 0>, rotation=<rot 90 0 0>}]
 dyn_zipper_mstr [{shapeType=Box, extentX=1.01, extentY=1.01, extentZ=1.01}]
 pack [{shapeType=Box, extentX=30.75, extentY=9.6169, extentZ=18.5, center=<0.5431 -1.9692 -12.2936>}]
}
constraints [
 {
  bone1Name=dyn_dagger_mstr_r
  bone2Name=pack
  swing1LimitDegrees=45
  swing1Motion=Free
  swing2LimitDegrees=45
  swing2Motion=Free
  twistLimitDegrees=45
  twistMotion=Free
 }
 {
  bone1Name=dyn_zipper_mstr
  bone2Name=pack
  swing1LimitDegrees=45
  swing1Motion=Free
  swing2LimitDegrees=45
  swing2Motion=Free
  twistLimitDegrees=45
  twistMotion=Free
 }
 {
  bone1Name=dyn_dagger_mstr_l
  bone2Name=pack
  swing1LimitDegrees=45
  swing1Motion=Free
  swing2LimitDegrees=45
  swing2Motion=Free
  twistLimitDegrees=45
  twistMotion=Free
 }
 {bone1Name=dyn_dagger_l, bone2Name=dyn_dagger_mstr_l, swing1LimitDegrees=4, swing2LimitDegrees=6, twistLimitDegrees=1}
 {bone1Name=dyn_dagger_r, bone2Name=dyn_dagger_mstr_r, swing1LimitDegrees=4, swing2LimitDegrees=6, twistLimitDegrees=1}
 {bone1Name=dyn_zipper, bone2Name=dyn_zipper_mstr, swing1LimitDegrees=24.4286, swing2LimitDegrees=0, swing2Motion=Locked, twistLimitDegrees=5}
 {bone1Name=dyn_strap_01_r, bone2Name=pack, swing1LimitDegrees=13, swing2LimitDegrees=13, twistLimitDegrees=0}
 {bone1Name=dyn_strap_02_r, bone2Name=dyn_strap_01_r, swing1LimitDegrees=13, swing2LimitDegrees=13, twistLimitDegrees=0}
 {bone1Name=dyn_strap_03_r, bone2Name=dyn_strap_02_r, swing1LimitDegrees=13, swing2LimitDegrees=13, twistLimitDegrees=0}
 {bone1Name=dyn_strap_01_l, bone2Name=pack, swing1LimitDegrees=13, swing2LimitDegrees=13, twistLimitDegrees=0}
 {bone1Name=dyn_strap_02_l, bone2Name=dyn_strap_01_l, swing1LimitDegrees=13, swing2LimitDegrees=13, twistLimitDegrees=0}
 {bone1Name=dyn_strap_03_l, bone2Name=dyn_strap_02_l, swing1LimitDegrees=13, swing2LimitDegrees=13, twistLimitDegrees=0}
]
